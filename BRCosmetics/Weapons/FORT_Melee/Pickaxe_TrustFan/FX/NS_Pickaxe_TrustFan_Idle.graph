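# Idle Niagara effect for the TrustFan pickaxe, combining a flickering glow with short-lived sparks.
# Glow_Sword001 pins two spawn groups of sprite and mesh particles to skeletal mesh locations, pulsing size and color and fading alpha along the system Timeline.
# Sparks_Top spawns two opposite-velocity, noise-driven groups around a socket, only while System.ReacOn is set.
# Colors switch on User.bIsFrontEnd, with time-of-day tints for the glow in gameplay.
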NiagaraSystem /BRCosmetics/Weapons/FORT_Melee/Pickaxe_TrustFan/FX/NS_Pickaxe_TrustFan_Idle
dependencies {
 usedDataInterfaces [
  &NiagaraDataInterfaceColorCurve
  &NiagaraDataInterfaceCurve
  &NiagaraDataInterfaceVector2DCurve
  &NiagaraDataInterfaceSpriteRendererInfo
  &NiagaraDataInterfaceSkeletalMesh
 ]
 usedDynamicInputs [
  &Vector2DFromCurve
  &Boolean_OrOperation
  &InvertBool
 ]
 usedModules [
  &self:SystemSpawnScript.SystemScriptSource.SystemScriptGraph.NiagaraNodeAssignment_0.SetVariables_#1
  &SystemState
  &self:SystemSpawnScript.SystemScriptSource.SystemScriptGraph.NiagaraNodeAssignment_2.SetVariables_#2
  &Timeline
 ]
}
emitters {
 Glow_Sword001 {
  emitterUpdateScript [
   &EmitterState(Life Cycle Mode=::NewEnumerator1, Loop Behavior=::NewEnumerator1, Loop Duration Mode=::NewEnumerator1, UseLoopDelay=false)
   &SpawnRate(SpawnRate=1.5, Use Spawn Probability=false)
   &SpawnRate(SpawnGroup=1, SpawnRate=1.5, Use Spawn Probability=false) {module=SpawnRate001}
  ]
  particleSpawnScript [
   /SetVariables(Particles.VisibilityTag=&SpawnGroup()) {isSetParameters=true}
   &InitializeParticle(
    Color=&MergeHelper_ColorSwitch(
     0=&MakeCustomLinearColorFromBool(
      Bool=$User.bIsFrontEnd
      True Color=<rgba 1 1 1 1>
      False Color=&FNTimeOfDayColor(Night Color=<rgba 0.85 0.85 0.85 0.7>, Morning Color=<rgba 1 1 1 0.6>, Day Color=<rgba 1 1 1 0.7>, Evening Color=<rgba 1 1 1 0.7>)
     )
     1=&MakeCustomLinearColorFromBool(
      Bool=$User.bIsFrontEnd
      True Color=<rgba 0.7 0.7 0.7 0.6>
      False Color=&FNTimeOfDayColor(Night Color=<rgba 1.3 1.3 1.3 1>, Morning Color=<rgba 1.5 1.5 1.5 1>, Day Color=<rgba 1.5 1.5 1.5 1>, Evening Color=<rgba 1.5 1.5 1.5 1>)
     )
     2=?
     3=?
     Index=&SpawnGroup()
     Num=::NewEnumerator0
    )
    Lifetime=&MergeHelper_FloatSwitch(0=1, 1=1, 2=?, 3=?, Index=&SpawnGroup(), Num=::NewEnumerator0)
    Mesh Scale=<1 1.4 0.9>
    Mesh Scale Mode=::NewEnumerator4
    Sprite Rotation Angle=5
    Sprite Rotation Mode=::NewEnumerator0
    Sprite Size=<37 115>
    Sprite Size Mode=::NewEnumerator4
    UsePositionOffset=false
    Write Material Random=true
   )
   &SpriteFacingAndAlignment(
    Alignment Coordinate Space=::Local
    Facing Coordinate Space=::Local
    Sprite Alignment=<0 0 1>
    Sprite Facing=<0 1 0>
    WriteAlignment=true
    WriteFacing=true
   )
   &InitialMeshOrientation(
    Mesh Axis To Orient=<0 1 0>
    Mesh Orientation Mode=::NewEnumerator2
    Orientation Vector=<0 1 0>
    Rotation=<0 0.055 0>
    Rotation Coordinate Space=::NewEnumerator2
    Use Rotation Vector=true
   )
  ]
  particleUpdateScript [
   &ParticleState()
   &SkeletalMeshLocation(
    Bone / Socket Index=&MergeHelper_IntSwitch(0=1, 1=0, 2=?, 3=?, 4=?, 5=?, Index=&SpawnGroup(), Num=::NewEnumerator0)
    MaskSpawnGroup=false
    Mesh Sampling Type=::NewEnumerator5
    Orientation Sampling=::NewEnumerator1
    Skeletal Mesh=$System.Skeletal Mesh
    Socket Sampling Mode=::NewEnumerator5
   )
   &ScaleMeshSize(Scale Factor=&VectorFromFloat(Value=&Waveform("Waveform [1]"=::NewEnumerator0, "[1] Frequency"=6, "[1] Amplitude Min/Max"=<1 1.05>)))
   &ScaleSpriteSize(
    Scale Sprite Size Mode=::NewEnumerator0
    Uniform Scale Factor=&Waveform("Waveform [1]"=::NewEnumerator0, "[1] Frequency"=6, "[1] Amplitude Min/Max"=<1.1 1.18>)
   )
   &SubUVAnimation(
    Direct Frame Index:NiagaraFloat=&FloatFromCurve(FloatCurve=DataInterface() {propertyValues {Curve {keys [<at 0 0>,<at 1 1>]}}}, Scale Curve=13)
    "Sprite Renderer "=DataInterface() {propertyValues {SpriteRenderer=&self:Glow_Sword001_0.NiagaraSpriteRendererProperties_0}}
    SubUV Animation Mode:ENiagara_SubUVLookupModeV2=::NewEnumerator5
   )
   &ScaleColor(
    Scale Alpha=&Multiply_Float(
     A=&MergeHelper_FloatSwitch(
      0=&Waveform("Waveform [1]"=::NewEnumerator0, "[1] Frequency"=5, "[1] Amplitude Min/Max"=<0.7 1>)
      1=&Waveform(
       "Waveform [1]"=::NewEnumerator0
       "[1] Frequency"=3
       "[1] Amplitude Min/Max"=&FN_FrontEndVector2Modifier(Front End:Vector2f=<0.5 1>, Gameplay:Vector2f=<0.2 1>)
      )
      2=?
      3=?
      Index=&SpawnGroup()
      Num=::NewEnumerator0
     )
     B=&FloatFromCurve(FloatCurve=DataInterface() {propertyValues {Curve {
         defaultValue=1
         keys [
          {arriveTangent=-0.0224, interpMode=RCIM_Cubic, leaveTangent=-0.0224, leaveTangentWeight=0.2653, tangentMode=RCTM_User, time=0, value=1}
          {
           arriveTangent=-1.1179
           arriveTangentWeight=0.3979
           interpMode=RCIM_Cubic
           leaveTangent=-1.1179
           leaveTangentWeight=0.1021
           tangentMode=RCTM_User
           time=0.7958
           value=0.6228
          }
          {arriveTangent=-4.1785, arriveTangentWeight=0.2925, interpMode=RCIM_Cubic, leaveTangent=-4.1785, tangentMode=RCTM_User, time=1, value=0}
         ]
        }}})
    )
    Scale RGB=&VectorFromFloat(Value=&FloatFromCurve(FloatCurve=DataInterface() {propertyValues {Curve {
         defaultValue=1
         keys [
          {interpMode=RCIM_Cubic, tangentMode=RCTM_User, time=0, value=1}
          {interpMode=RCIM_Cubic, tangentMode=RCTM_User, time=0.05, value=1.5}
          <at 0.1 1 RCIM_Cubic>
          <at 0.15 1.5 RCIM_Cubic>
          <at 0.2 1 RCIM_Cubic>
          <at 0.25 1.5 RCIM_Cubic>
          <at 0.3 1 RCIM_Cubic>
          <at 0.35 1.5 RCIM_Cubic>
          <at 0.4 1 RCIM_Cubic>
          <at 0.45 1.5 RCIM_Cubic>
          {arriveTangent=3.0958, interpMode=RCIM_Cubic, leaveTangent=3.0958, tangentMode=RCTM_User, time=0.5, value=1}
          <at 0.55 1.5 RCIM_Cubic>
          <at 0.6 1 RCIM_Cubic>
          <at 0.653 1.5 RCIM_Cubic>
          <at 0.7 1 RCIM_Cubic>
          <at 0.753 1.5 RCIM_Cubic>
          <at 0.8054 1 RCIM_Cubic>
          <at 0.8692 1.5 RCIM_Cubic>
          <at 0.9192 1 RCIM_Cubic>
          <at 0.9648 1.5 RCIM_Cubic>
          {interpMode=RCIM_Cubic, leaveTangent=-0.0855, time=0.9951, value=1}
         ]
        }}}))
    ScaleA=true
    ScaleRGB=true
   )
   &CameraOffset(Camera Offset Amount=&MergeHelper_FloatSwitch(0=-10, 1=-15, 2=?, 3=?, Index=&SpawnGroup(), Num=::NewEnumerator0))
   &ScaleColor(
    Scale Alpha=&FloatFromCurve(
     FloatCurve=DataInterface() {propertyValues {Curve {
        defaultValue=1
        keys [
         {arriveTangent=-2.4146, interpMode=RCIM_Cubic, leaveTangent=-2.4146, tangentMode=RCTM_User, time=0, value=1}
         {interpMode=RCIM_Cubic, tangentMode=RCTM_User, time=1, value=0}
        ]
       }}}
     CurveIndex=$System.Timeline.NormalizedTime
    )
    ScaleA=true
    ScaleRGB=true
   ) {module=ScaleColor001}
  ]
  rendererClasses [&NiagaraSpriteRendererProperties, &NiagaraMeshRendererProperties]
  renderers [
   {rendererClass=&NiagaraSpriteRendererProperties}
   {rendererClass=&NiagaraMeshRendererProperties, rendererIndex=1}
  ]
 }
 Sparks_Top {
  emitterUpdateScript [
   &EmitterState(
    Inactive Response=::NewEnumerator1
    Life Cycle Mode=::NewEnumerator1
    Loop Behavior=::NewEnumerator1
    Loop Duration Mode=::NewEnumerator1
    UseLoopDelay=false
   )
   &SpawnRate(
    SpawnRate=&MakeCustomFloatFromBool(Bool=$System.ReacOn, True Float=7, False Float=0)
    Use Spawn Probability=false
   ) {module=SpawnRate001}
   &SpawnRate(
    SpawnGroup=1
    SpawnRate=&MakeCustomFloatFromBool(Bool=$System.ReacOn, True Float=7, False Float=0)
    Use Spawn Probability=false
   )
  ]
  particleSpawnScript [
   /SetVariables(Particles.VisibilityTag=&SpawnGroup()) {isSetParameters=true}
   &InitializeParticle(
    Color=&MakeCustomLinearColorFromBool(Bool=$User.bIsFrontEnd, True Color=<rgba 1 1 1 0.3>, False Color=<rgba 2.5 2.5 2.5 1>)
    Lifetime Max=0.1
    Lifetime Min=0.25
    Lifetime Mode=::NewEnumerator1
    Sprite Rotation Mode=::NewEnumerator2
    Sprite Size Max=<1.3 9>
    Sprite Size Min=<2.5 6>
    Sprite Size Mode=::NewEnumerator2
    UsePositionOffset=false
    Write Material Random=true
   )
   &CameraOffset(Camera Offset Amount=-5)
   &SocketLocation(
    Interpolate Socket Sampling=true
    SamplePosOffset=true
    Sampled Position Offset=&MergeHelper_VectorSwitch(0=<-2 0 -28>, 1=<-8 0 -28>, 2=?, 3=?, Index=&SpawnGroup(), Num=::NewEnumerator0)
    Socket Reader=$System.Socket Reader
    Socket Sampling Mode=::NewEnumerator3
    VelocitySpeedLimit=false
   )
   &ShapeLocation(
    Cylinder Height=90
    Cylinder Height Midpoint=-0.1
    Cylinder Radius=3
    MaskSpawnGroup=false
    Override Seed=false
    Rotation Axis=<0.15 0 1>
    Rotation Mode=::NewEnumerator1
    Shape Primitive=::NewEnumerator1
    Surface Only=false
    UseOwnerScale=true
   )
   &AddVelocity(
    Cone Angle=180
    Module.UseConeFalloff=false
    Module.UseVelDistribution=false
    Override Seed=false
    Velocity Mode=::NewEnumerator2
    Velocity Speed=&MergeHelper_FloatSwitch(
     0=&RandomRangeFloat(Minimum=150, Maximum=100)
     1=&RandomRangeFloat(Minimum=-150, Maximum=-100)
     2=?
     3=?
     Index=&SpawnGroup()
     Num=::NewEnumerator0
    )
   ) {module=AddVelocity001}
  ]
  particleUpdateScript [
   &ParticleState()
   &ScaleColor(
    Scale Alpha=&FloatFromCurve(FloatCurve=DataInterface() {propertyValues {Curve {
        defaultValue=1
        keys [
         {arriveTangent=5.0181, interpMode=RCIM_Cubic, leaveTangent=5.0181, tangentMode=RCTM_User, time=0, value=0}
         <at 0.3 1 RCIM_Cubic>
         {arriveTangent=-0.9917, interpMode=RCIM_Cubic, leaveTangent=-0.9917, tangentMode=RCTM_User, time=1, value=0}
        ]
       }}})
    ScaleA=true
    ScaleRGB=true
   ) {module=ScaleColor001}
   &ScaleSpriteSize(Scale Factor=&Vector2DFromFloat(Value=&FloatFromCurve(FloatCurve=DataInterface() {propertyValues {Curve {keys [
          {arriveTangent=10.1867, interpMode=RCIM_Cubic, leaveTangent=10.1867, tangentMode=RCTM_User, time=0, value=0}
          {interpMode=RCIM_Cubic, leaveTangent=-0.0263, tangentMode=RCTM_Break, time=0.2, value=1.2}
          {arriveTangent=-2.7631, interpMode=RCIM_Cubic, leaveTangent=-2.7631, tangentMode=RCTM_User, time=1, value=0}
         ]}}}))) {module=ScaleSpriteSize001}
   &CurlNoiseForce(Noise Frequency=100, Noise Strength=500, Pan Noise=true, Pan Noise Field=<0 0 0.5>, Random Seed=10)
   &Drag(Drag=&RandomRangeFloat(Minimum=5, Maximum=6), Use Linear Drag=true, Use Rotational Drag=false)
   &GravityForce(Gravity=<0 0 200>)
   &InheritSourceMovement(
    Applied Position Scale=&VectorFromFloat(Value=&FloatFromCurve(FloatCurve=DataInterface() {propertyValues {Curve {
         defaultValue=1
         keys [
          {interpMode=RCIM_Cubic, tangentMode=RCTM_User, time=0, value=1}
          {arriveTangent=-2.2715, interpMode=RCIM_Cubic, leaveTangent=-2.2715, tangentMode=RCTM_User, time=1, value=0}
         ]
        }}}))
    ApplyForce=false
    ApplyPosition=true
    ApplyPositionFalloff=false
   )
   &SolveForcesAndVelocity(Clamp Velocity=false, Limit Acceleration=false)
   &ScaleSpriteSizeBySpeed(Max Scale Factor=<0.8 2>, Velocity Threshold=300)
  ]
  rendererClasses [&NiagaraSpriteRendererProperties]
  renderers [
   {rendererClass=&NiagaraSpriteRendererProperties}
   {rendererClass=&NiagaraSpriteRendererProperties, rendererIndex=1}
  ]
 }
}
summary {
 systemName=NS_Pickaxe_TrustFan_Idle
 emitters [
  {emitterName=Sparks_Top, rendererClasses [&NiagaraSpriteRendererProperties]}
  {emitterName=Glow_Sword001, rendererClasses [&NiagaraSpriteRendererProperties, &NiagaraMeshRendererProperties]}
 ]
 userVariables [
  User.bIsFrontEnd {type=NiagaraBool}
  User.bIsFrontEndPreview {type=NiagaraBool}
  User.HitWeakSpot {type=NiagaraBool}
  User.ReactivityOn {type=NiagaraBool}
 ]
}
